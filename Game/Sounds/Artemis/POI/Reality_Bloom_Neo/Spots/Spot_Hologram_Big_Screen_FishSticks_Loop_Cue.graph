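# Sound cue for a looping ambient spot on the FishSticks big-screen hologram in the Reality Bloom Neo POI, routed to the outdoor ambient sound class.
# A Mixer layers two looping Neo hologram waves at equal volume.
# The root Quality Level node sends its first two slots to that Mixer and leaves the other two empty.
# A shared forward-cone attenuation asset shapes falloff, and overridden concurrency caps it at two instances.
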SoundCue /Game/Sounds/Artemis/POI/Reality_Bloom_Neo/Spots/Spot_Hologram_Big_Screen_FishSticks_Loop_Cue
root=QualityLevel_0
nodes [
 {id=Mixer_1, children [WavePlayer_17, WavePlayer_20], properties {inputVolume [1,1]}}
 {id=QualityLevel_0, children [Mixer_1,Mixer_1,null,null]}
 {id=WavePlayer_17, properties {bLooping=true, soundWaveAssetPtr=&/Spot_Neo_Hologram_Loop_07}}
 {id=WavePlayer_20, properties {bLooping=true, soundWaveAssetPtr=&/Spot_Neo_Hologram_Loop_12}}
]
properties {
 attenuationSettings=/Game/Sounds/Artemis/POI/Reality_Bloom_Neo/Spot_Hologram_Big_Screen_Forward_Cone_Sphere
 bOverrideConcurrency=true
 maxDistance=3800
 soundClassObject=&Fort_Ambient_Outdoors_Athena
 volumeMultiplier=0.401
 attenuationOverrides {
  bEnableFocusInterpolation=true
  falloffDistance=600
  nonFocusVolumeAttenuation=0.3
  occlusionVolumeAttenuation=0.3
  stereoSpread=0
  attenuationShapeExtents {x=200}
 }
 concurrencyOverrides {maxCount=2}
}
