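# Ambient spot Sound Cue for the rotating antenna hologram on a big screen at the Reality Bloom Neo POI.
# A quality-level root routes all four slots to one wave player playing a Neo hologram loop wave, so every quality sounds the same.
# It uses the outdoor ambient sound class and a shared big-screen sphere attenuation asset, with its own concurrency override capping simultaneous instances at two.
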SoundCue /Game/Sounds/Artemis/POI/Reality_Bloom_Neo/Spots/Spot_Hologram_Big_Screen_Rotating_Antenna_Loop_Cue
root=QualityLevel_0
nodes [
 {id=QualityLevel_0, children [WavePlayer_0,WavePlayer_0,WavePlayer_0,WavePlayer_0]}
 {id=WavePlayer_0, properties {soundWaveAssetPtr=&/Spot_Neo_Hologram_Loop_14}}
]
properties {
 attenuationSettings=/Game/Sounds/Artemis/POI/Reality_Bloom_Neo/Spot_Hologram_Big_Screen_Sphere
 bOverrideConcurrency=true
 maxDistance=5250
 soundClassObject=&Fort_Ambient_Outdoors_Athena
 volumeMultiplier=1.002
 attenuationOverrides {
  bEnableFocusInterpolation=true
  falloffDistance=600
  nonFocusVolumeAttenuation=0.3
  occlusionVolumeAttenuation=0.3
  stereoSpread=0
  attenuationShapeExtents {x=200}
 }
 concurrencyOverrides {maxCount=2}
}
