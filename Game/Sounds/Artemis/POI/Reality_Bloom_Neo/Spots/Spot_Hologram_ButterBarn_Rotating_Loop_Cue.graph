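# Looping ambient spot sound for the rotating ButterBarn hologram at the Reality Bloom Neo POI, in the outdoor ambient sound class.
# A looping hologram wave feeds a mixer through a large-sphere attenuation node, while the mixer's other attenuation branch has no sound connected.
# A quality node routes that mixer into its first two slots and leaves two empty, under an outer loop node.
# The cue also overrides concurrency to limit simultaneous instances.
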SoundCue /Game/Sounds/Artemis/POI/Reality_Bloom_Neo/Spots/Spot_Hologram_ButterBarn_Rotating_Loop_Cue
root=Looping_1
nodes [
 {
  id=Attenuation_0
  children [null]
  properties {attenuationSettings=/Game/Sounds/Artemis/POI/Reality_Bloom_Neo/Spot_Hologram_Big_Screen_Sphere}
 }
 {
  id=Attenuation_1
  children [WavePlayer_0]
  properties {attenuationSettings=/Game/Sounds/Artemis/POI/Reality_Bloom_Neo/Spot_Hologram_Big_Screen_Large_Sphere}
 }
 {id=Looping_1, children [QualityLevel_0]}
 {id=Mixer_0, children [Attenuation_0, Attenuation_1], properties {inputVolume [0.3,2]}}
 {id=QualityLevel_0, children [Mixer_0,Mixer_0,null,null]}
 {id=WavePlayer_0, properties {bLooping=true, soundWaveAssetPtr=&/Spot_Neo_Hologram_Loop_14}}
]
properties {
 attenuationSettings=&Spot_Hologram_Big_Screen_Sphere
 bOverrideConcurrency=true
 maxDistance=5250
 soundClassObject=&Fort_Ambient_Outdoors_Athena
 volumeMultiplier=0.451
 attenuationOverrides {
  bEnableFocusInterpolation=true
  falloffDistance=600
  nonFocusVolumeAttenuation=0.3
  occlusionVolumeAttenuation=0.3
  stereoSpread=0
  attenuationShapeExtents {x=200}
 }
 concurrencyOverrides {maxCount=2}
}
