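# Looping spot ambience cue for an interior hologram glass screen at the Reality Bloom Neo point of interest.
# A single hologram loop wave is pitched down by a fixed Modulator, then drifts slowly around that pitch through an indefinitely looping two-minute Enveloper curve.
# An outer Enveloper adds a short fade-in at start.
# It borrows the server stack attenuation and overrides concurrency to allow two voices.
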SoundCue /Game/Sounds/Artemis/POI/Reality_Bloom_Neo/Spots/Spot_Hologram_Glass_Screen_Loop_Interior_01_Cue
root=Enveloper_29
nodes [
 {
  id=Enveloper_27
  children [Modulator_1]
  properties {
   bLoop=true
   bLoopIndefinitely=true
   loopEnd=120
   pitchCurve {editorCurveData {keys [
      {interpMode=RCIM_Cubic, time=0}
      <at 30 0.95 RCIM_Cubic>
      {arriveTangent=0.0017, interpMode=RCIM_Cubic, leaveTangent=0.0017, time=60}
      <at 90 1.05 RCIM_Cubic>
      {interpMode=RCIM_Cubic, time=120}
     ]}}
   volumeCurve {editorCurveData {keys [<at 0 1>]}}
  }
 }
 {
  id=Enveloper_29
  children [Enveloper_27]
  properties {
   pitchCurve {editorCurveData {keys [{time=0}]}}
   volumeCurve {editorCurveData {keys [{arriveTangent=3.4293, interpMode=RCIM_Cubic, leaveTangent=3.4293, tangentMode=RCTM_User, time=0, value=0}, <at 0.4 1 RCIM_Cubic>]}}
  }
 }
 {id=Modulator_1, children [WavePlayer_0], properties {pitchMax=0.8, pitchMin=0.8}}
 {id=WavePlayer_0, properties {soundWaveAssetPtr=&/Spot_Neo_Hologram_Loop_13}}
]
properties {
 attenuationSettings=/Game/Sounds/Artemis/Environment_Actors/Server/Spot_Server_Stack_Attenuation
 bEnableBusSends=false
 bEnableSubmixSends=false
 bOverrideConcurrency=true
 maxDistance=562
 soundClassObject=&Fort_Gadgets
 volumeMultiplier=0.8
 concurrencyOverrides {maxCount=2, voiceStealReleaseTime=0.2}
}
