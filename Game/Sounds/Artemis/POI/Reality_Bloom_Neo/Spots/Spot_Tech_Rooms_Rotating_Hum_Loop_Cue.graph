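# Looping spot ambience cue for a rotating tech-room hum at the Reality Bloom Neo POI, reusing an occluded Hollow Base loop wave.
# An outer envelope with a short volume fade-in feeds a random start delay, then an endlessly looping envelope that slowly sways pitch down and up.
# A modulator adds random variation beneath it.
# Hum attenuation, the gadgets sound class and a tight concurrency cap govern playback.
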SoundCue /Game/Sounds/Artemis/POI/Reality_Bloom_Neo/Spots/Spot_Tech_Rooms_Rotating_Hum_Loop_Cue
root=Enveloper_29
nodes [
 {id=Delay_0, children [Enveloper_27], properties {delayMax=0.75, delayMin=0.05}}
 {
  id=Enveloper_27
  children [Modulator_1]
  properties {
   bLoop=true
   bLoopIndefinitely=true
   loopEnd=120
   pitchCurve {editorCurveData {keys [
      {interpMode=RCIM_Cubic, time=0}
      <at 30 0.95 RCIM_Cubic>
      {arriveTangent=0.0017, interpMode=RCIM_Cubic, leaveTangent=0.0017, time=60}
      <at 90 1.05 RCIM_Cubic>
      {interpMode=RCIM_Cubic, time=120}
     ]}}
   volumeCurve {editorCurveData {keys [<at 0 1>]}}
  }
 }
 {
  id=Enveloper_29
  children [Delay_0]
  properties {
   pitchCurve {editorCurveData {keys [{time=0}]}}
   volumeCurve {editorCurveData {keys [{arriveTangent=3.4293, interpMode=RCIM_Cubic, leaveTangent=3.4293, tangentMode=RCTM_User, time=0, value=0}, <at 0.4 1 RCIM_Cubic>]}}
  }
 }
 {id=Modulator_1, children [WavePlayer_2], properties {pitchMax=1}}
 {id=WavePlayer_2, properties {bLooping=true, soundWaveAssetPtr=/Game/Sounds/Artemis/POI/Hollow_Base/Spots/Spot_Tech_Rooms_Occluded_Loop_04}}
]
properties {
 attenuationSettings=/Game/Sounds/Artemis/POI/Reality_Bloom_Neo/Spot_Neo_Ambient_Hum_Attenuation
 bEnableBusSends=false
 bEnableSubmixSends=false
 bOverrideConcurrency=true
 maxDistance=750
 soundClassObject=&Fort_Gadgets
 volumeMultiplier=0.6
 concurrencyOverrides {maxCount=2, voiceStealReleaseTime=0.2}
}
